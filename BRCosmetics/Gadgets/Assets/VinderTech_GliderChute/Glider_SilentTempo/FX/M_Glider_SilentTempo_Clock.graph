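# Translucent unlit material for the clock effect on the Silent Tempo glider, built for Niagara mesh particles.
# Packed numeral, outline and symbol channels rotate via dynamic parameters, combined with curve-atlas gradients, an animated radial sweep and time-of-day desaturation.
# Vertex red blends normal-based inner and outer radius displacement plus a local-Z offset.
# Opacity combines particle alpha, a two-sided-sign mask and a Perlin-broken angular erosion with controllable direction and progress.
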Material /BRCosmetics/Gadgets/Assets/VinderTech_GliderChute/Glider_SilentTempo/FX/M_Glider_SilentTempo_Clock
graph {
 properties {
  bUsedWithNiagaraMeshParticles=true
  blendMode=BLEND_Translucent
  pixelDepthOffsetMode=PDOM_Legacy
  shadingModel=MSM_Unlit
  twoSided=true
  usedShadingModels=Unlit
 }
 nodes [
  n1 = DynamicParameter() {parameterIndex=2, paramNames [InsideRadiusUpdate, OutsideRadiusUpdate, "", ""]}
  n2 = Multiply(A=ScalarParameter() {defaultValue=20, group=Shape, parameterName=OutSide Raidus}, B=@n1.OutsideRadiusUpdate)
  n3 = Multiply(A=ScalarParameter() {defaultValue=55, group=Shape, parameterName=Inside Radius}, B=@n1.InsideRadiusUpdate)
  n4 = VertexColor()
  n5 = Multiply(A=ScalarParameter() {defaultValue=3, group=Shape, parameterName=Inside Offset}, B=Constant4Vector() {constant=<rgba 0 0 1 0>}.RGBA)
  Add(
   A=Multiply(A=VertexNormalWS(), B=LinearInterpolate(A=@n2, B=@n3, Alpha=@n4.R))
   B=LinearInterpolate(B=Transform(None=@n5) {transformSourceType=TRANSFORMSOURCE_Local}, Alpha=@n4.R) {constB=0}
  )
  n6 = MaterialFunctionCall(Float2=MaterialFunctionCall() {materialFunction=&VectorToRadialValue}.Radial Coordinates) {materialFunction=&BreakOutFloat2Components}
  n7 = OneMinus(None=ConstantBiasScale(None=Cosine(None=@n6.R)))
  n8 = Saturate(None=Add(A=ScalarParameter() {group=Sweep, parameterName=Sweep Range}, B=SquareRoot(None=@n7)) {constB=0.3})
  n9 = DynamicParameter() {defaultValue=<rgba 1 0.5 0.5 1>, paramNames [Number Rotator, Symbol Rotator, Sweep Age, TwoSide]}
  n10 = Add(A=OneMinus(None=@n6.R), B=LinearInterpolate(Alpha=Frac(None=@n9.Sweep Age)) {constA=-1})
  n11 = CurveAtlasRowParameter(CurveTime=@n10) {
   atlas=/Game/Athena/LinearColorCurves/CA_Contrail_SilentTempo
   curve=/Game/Gadgets/Assets/VinderTech_GliderChute/Glider_SilentTempo/FX/Curve_Glider_SilentTempo_03_Peril
   defaultValue=6
   parameterName=CurveB
  }
  n12 = Multiply(A=Multiply(A=Multiply(A=@n8, B=@n10), B=@n6.G), B=@n11.R)
  n13 = DynamicParameter() {defaultValue=<rgba 1 0.5 0.5 1>, paramNames [Number Rotator, Symbol Rotator, Sweep Age, TwoSide]}
  n14 = Rotator(
   Coordinate=MaterialFunctionCall(UVs=TextureCoordinate() {coordinateIndex=1}, Texture Scale=ScalarParameter() {defaultValue=0.95, group=Element, parameterName=Number Scale}) {materialFunction=&ScaleUVsByCenter}.UVs
   Time=Multiply(A=@n13.Number Rotator, B=@n13.Number Rotator) {constB=-1}
  ) {speed=2}
  n15 = TextureSample(UVs=@n14) {
   samplerSource=SSM_Clamp_WorldGroupSettings
   texture=/BRCosmetics/Athena/Effects/Particles/Skydiving/NS_Contrail_SlientTempo/T_SilentTempo_Nurmerals
  }
  n16 = VectorParameter() {group=Element, parameterName=Number Color, defaultValue=<rgba 0.3543 0.89 1 1>}
  n17 = TextureSample(UVs=@n14) {
   samplerSource=SSM_Clamp_WorldGroupSettings
   texture=&T_SilentTempo_Nurmerals
  }
  n18 = VectorParameter() {group=Element, parameterName=Outline Color, defaultValue=<rgba 0 0.1667 1 1>}
  n19 = TextureCoordinate() {coordinateIndex=1}
  n20 = ScalarParameter() {defaultValue=1, group=Element, parameterName=Symbol Scale}
  n21 = TextureSample(UVs=Rotator(
    Coordinate=MaterialFunctionCall(UVs=@n19, Texture Scale=@n20) {materialFunction=&ScaleUVsByCenter}.UVs
    Time=Multiply(A=@n13.Symbol Rotator, B=@n13.Symbol Rotator) {constB=-1}
   ) {speed=-0.25}) {
   samplerSource=SSM_Clamp_WorldGroupSettings
   texture=&T_SilentTempo_Nurmerals
  }
  n22 = Add(
   A=Add(A=Multiply(A=@n15.R, B=@n16.RGB), B=Multiply(A=@n17.B, B=@n18.RGB))
   B=Multiply(A=@n21.G, B=VectorParameter() {group=Element, parameterName=Symbol Color, defaultValue=<rgba 0.0101 0.1471 0.25 1>}.RGB)
  )
  n23 = CurveAtlasRowParameter(CurveTime=MaterialFunctionCall(Float2=MaterialFunctionCall() {materialFunction=&VectorToRadialValue}.Radial Coordinates) {materialFunction=&BreakOutFloat2Components}.G) {
   atlas=&CA_Contrail_SilentTempo
   curve=/Game/Gadgets/Assets/VinderTech_GliderChute/Glider_SilentTempo/FX/Curve_Glider_SilentTempo_01_Peril
   defaultValue=4
   parameterName=CurveA
  }
  n24 = Add(
   A=Multiply(A=Multiply(A=@n12, B=@n22), B=ScalarParameter() {defaultValue=30, group=Sweep, parameterName=Sweep Intensity})
   B=Add(A=@n22, B=@n23)
  )
  n25 = MaterialFunctionCall(
   Value_Morning=ScalarParameter() {group=TOD, parameterName=Tod Desaturation Morning}
   Value_Day=ScalarParameter() {group=TOD, parameterName=Tod Desaturation Day}
   Value_Evening=ScalarParameter() {group=TOD, parameterName=Tod Desaturation Evening}
   Value_Night=ScalarParameter() {defaultValue=0.2, group=TOD, parameterName=Tod Desaturation Night}
  ) {materialFunction=&Scalar_Time_of_Day}
  n26 = ParticleColor()
  MaterialFunctionCall(Emissive=Multiply(A=Desaturation(None=@n24, Fraction=@n25.Result), B=@n26.RGB)) {materialFunction=&MF_FX_EmissiveAdjustments}
  n27 = Multiply(
   A=TwoSidedSign()
   B=DynamicParameter() {defaultValue=<rgba 1 0.5 0.5 1>, paramNames [Number Rotator, Symbol Rotator, Sweep Age, TwoSide]}.TwoSide
  ) {constB=-1}
  n28 = MaterialFunctionCall() {materialFunction=&VectorToRadialValue}
  n29 = Add(
   A=MaterialFunctionCall(Float2=@n28.Radial Coordinates) {materialFunction=&BreakOutFloat2Components}.R
   B=ScalarParameter() {defaultValue=0.5, group=Erode, parameterName=Erode Start Angle}
  ) {constB=0}
  n30 = Cosine(None=@n29)
  n31 = SquareRoot(None=OneMinus(None=ConstantBiasScale(None=@n30)))
  n32 = DynamicParameter() {parameterIndex=1, defaultValue=<rgba 1 0 1 1>, paramNames [Erode Direction, Erode Edge Mask, Erode Progress, ""]}
  n33 = Frac(None=@n29)
  n34 = LinearInterpolate(A=OneMinus(None=@n33), B=@n33, Alpha=@n32.Erode Direction)
  n35 = TextureSampleParameter2D(UVs=Multiply(A=@n28.Radial Coordinates, B=Constant2Vector() {g=0.5, r=1})) {
   parameterName=Erode Tex
   samplerType=SAMPLERTYPE_LinearGrayscale
   texture=&T_SimplePerlin_Greyscale
  }
  n36 = LinearInterpolate(A=@n34, B=Multiply(A=@n34, B=@n35.R), Alpha=ScalarParameter() {defaultValue=0.5, group=Erode, parameterName=Lerp Erode Tex})
  n37 = ScalarParameter() {defaultValue=2, group=Erode, parameterName=Erode Edge Hardness, sliderMin=0.5}
  n38 = Multiply(
   A=Saturate(None=Power(Base=@n31, Exp=@n32.Erode Edge Mask) {constExponent=1})
   B=Subtract(A=Multiply(A=@n36, B=@n37), B=LinearInterpolate(A=@n37, Alpha=@n32.Erode Progress) {constB=-1})
  )
  Multiply(A=@n26.A, B=Multiply(A=Clamp(None=@n27), B=Saturate(None=@n38)))
 ]
}
