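# Translucent, unlit, two-sided glow material for Niagara sprite and mesh particles on the Silent Tempo glider.
# Emissive color comes from a curve atlas row sampled across U, or a flat color, tinted by particle color and dimmed on low/medium quality, forward and mobile paths.
# Opacity uses a switchable sphere mask or curve-alpha/U-gradient blend, scaled by particle alpha and view-angle fade.
# Depth fade softens intersections except on mobile.
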Material /BRCosmetics/Gadgets/Assets/VinderTech_GliderChute/Glider_SilentTempo/FX/M_Glider_SilentTempo_Glow
graph {
 properties {
  bUsedWithNiagaraMeshParticles=true
  bUsedWithNiagaraSprites=true
  blendMode=BLEND_Translucent
  pixelDepthOffsetMode=PDOM_Legacy
  shadingModel=MSM_Unlit
  twoSided=true
  usedShadingModels=Unlit
 }
 nodes [
  n1 = CurveAtlasRowParameter(CurveTime=MaterialFunctionCall(Float2=TextureCoordinate()) {materialFunction=&BreakOutFloat2Components}.R) {
   atlas=/Game/Athena/LinearColorCurves/CA_Contrail_SilentTempo
   curve=/Game/Gadgets/Assets/VinderTech_GliderChute/Glider_SilentTempo/FX/Curve_Glider_SilentTempo_05_Peril
   defaultValue=8
   parameterName=Curve
  }
  n2 = ParticleColor()
  n3 = Multiply(
   A=StaticSwitchParameter(True=@n1, False=VectorParameter() {parameterName=Coloe, defaultValue=<rgba 1 1 1 1>}.RGB) {defaultValue=true, parameterName=Use Color Curve}
   B=@n2.RGB
  )
  n4 = Multiply(A=@n3, B=ScalarParameter() {defaultValue=0.7, parameterName=Emissive Low})
  n5 = QualitySwitch(Default=@n3, Low=@n4, High=@n3, Medium=@n4, Epic=@n3)
  MaterialFunctionCall(Emissive=ShadingPathSwitch(Default=@n5, Deferred=@n5, Forward=@n4, Mobile=@n4)) {materialFunction=&MF_FX_EmissiveAdjustments}
  MaterialFunctionCall!(TexCoord, Tiles XY) {materialFunction=/Game/Environments/Papaya/Materials/MF_GridTexCoord}
  n6 = TextureCoordinate()
  n7 = Constant() {r=0.5}
  n8 = ScalarParameter() {defaultValue=0.4, parameterName=Sphere Radius}
  n9 = MaterialFunctionCall(Float2=TextureCoordinate()) {materialFunction=&BreakOutFloat2Components}
  n10 = Multiply(A=@n9.R, B=OneMinus(None=@n9.R))
  n11 = StaticSwitchParameter(
   True=Power(
    Base=SphereMask(A=@n6, B=@n7, Radius=@n8) {attenuationRadius=0.5, hardnessPercent=0}
    Exp=ScalarParameter() {defaultValue=3, parameterName=Sphere Power}
   )
   False=LinearInterpolate(A=@n1.A, B=Power(Base=@n10), Alpha=ScalarParameter() {group=Alpha, parameterName=Lerp Gradient Alpha})
  ) {parameterName="Use Sphere  Shape"}
  n12 = Multiply(
   A=Multiply(A=Multiply(A=@n2.A, B=@n11), B=ScalarParameter() {defaultValue=4, group=Alpha, parameterName=Opacity Mult})
   B=MaterialFunctionCall("FadeSensitivity (optional)"=ScalarParameter() {defaultValue=1, group=Alpha, parameterName=FadeViewAngle}) {materialFunction=&FadeBasedOnViewAngle}.Result
  )
  n13 = ScalarParameter() {defaultValue=25, group=Alpha, parameterName=FadeDistance}
  ShadingPathSwitch(Default=Multiply(A=@n12, B=DepthFade(FadeDistance=@n13)), Mobile=@n12) {desc=Migrated from FeatureLevelSwitch}
 ]
}
