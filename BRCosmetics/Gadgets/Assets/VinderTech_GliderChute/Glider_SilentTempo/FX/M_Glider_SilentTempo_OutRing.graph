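# Translucent, unlit, two-sided Niagara mesh-particle material for an outer ring effect on the Silent Tempo glider.
# A curve atlas row sampled by radial coordinate supplies color and alpha, plus an optional procedural sine-and-sphere-mask dot ring rotated by a dynamic parameter.
# Emissive gets time-of-day desaturation and particle color; particle alpha scales opacity.
# Vertices offset along normals, with red vertex color blending outside and inside radii, optionally driven by dynamic parameters.
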Material /BRCosmetics/Gadgets/Assets/VinderTech_GliderChute/Glider_SilentTempo/FX/M_Glider_SilentTempo_OutRing
graph {
 properties {
  bUsedWithNiagaraMeshParticles=true
  blendMode=BLEND_Translucent
  pixelDepthOffsetMode=PDOM_Legacy
  shadingModel=MSM_Unlit
  twoSided=true
  usedShadingModels=Unlit
 }
 nodes [
  n1 = DynamicParameter() {paramNames [Dot Rotator, "", InsideRadiusUpdate, OutsideRadiusUpdate]}
  Add(A=Time(), B=@n1.Dot Rotator)
  MaterialFunctionCall() {materialFunction=&DebugTimeSine}
  n2 = MaterialFunctionCall() {materialFunction=&VectorToRadialValue}
  n3 = CurveAtlasRowParameter(CurveTime=MaterialFunctionCall(Float2=@n2.Radial Coordinates) {materialFunction=&BreakOutFloat2Components}.G) {
   atlas=/Game/Athena/LinearColorCurves/CA_Contrail_SilentTempo
   curve=/Game/Gadgets/Assets/VinderTech_GliderChute/Glider_SilentTempo/FX/Curve_Glider_SilentTempo_02_Peril
   defaultValue=5
  }
  n4 = MaterialFunctionCall(Float2=@n2.Radial Coordinates) {materialFunction=&BreakOutFloat2Components}
  n5 = Multiply(A=AppendVector(A=Add(A=@n4.R, B=@n1.Dot Rotator), B=@n4.G), B=Constant2Vector() {g=9, r=48}) {constB=8}
  n6 = Constant2Vector() {g=0.5, r=0.5}
  n7 = SphereMask(A=ConstantBiasScale(None=Sine(None=@n5)), B=@n6) {attenuationRadius=0.2}
  n8 = TextureCoordinate()
  n9 = Saturate(None=Subtract(A=SphereMask(A=@n8, B=@n6) {attenuationRadius=0.48}, B=SphereMask(A=@n8, B=@n6) {attenuationRadius=0.45}))
  n10 = Add(A=@n3, B=LinearInterpolate(B=Multiply(A=@n7, B=@n9), Alpha=ScalarParameter() {defaultValue=1, parameterName=UseDotRing}))
  n11 = MaterialFunctionCall(
   Value_Morning=ScalarParameter() {group=TOD, parameterName=Tod Desaturation Morning}
   Value_Day=ScalarParameter() {group=TOD, parameterName=Tod Desaturation Day}
   Value_Evening=ScalarParameter() {group=TOD, parameterName=Tod Desaturation Evening}
   Value_Night=ScalarParameter() {defaultValue=0.2, group=TOD, parameterName=Tod Desaturation Night}
  ) {materialFunction=&Scalar_Time_of_Day}
  n12 = ParticleColor()
  MaterialFunctionCall(Emissive=Multiply(A=Desaturation(None=@n10, Fraction=@n11.Result), B=@n12.RGB)) {materialFunction=&MF_FX_EmissiveAdjustments}
  n13 = DynamicParameter() {paramNames [Dot Rotator, "", InsideRadiusUpdate, OutsideRadiusUpdate]}
  n14 = ScalarParameter() {parameterName=UseRadiusUpdate}
  n15 = Multiply(A=ScalarParameter() {defaultValue=10, parameterName=OutSideRaidus}, B=LinearInterpolate(B=@n13.OutsideRadiusUpdate, Alpha=@n14) {constA=1})
  n16 = Constant() {r=1}
  n17 = Multiply(A=ScalarParameter() {defaultValue=30, parameterName=InsideRadius}, B=LinearInterpolate(B=@n13.InsideRadiusUpdate, Alpha=@n14) {constA=1})
  Multiply(
   A=VertexNormalWS()
   B=LinearInterpolate(A=Multiply(A=@n16, B=@n15) {constB=25}, B=Multiply(A=@n16, B=@n17) {constB=15}, Alpha=VertexColor().R)
  )
  Multiply(A=Multiply(A=@n3.A) {constB=0.5}, B=@n12.A)
 ]
}
