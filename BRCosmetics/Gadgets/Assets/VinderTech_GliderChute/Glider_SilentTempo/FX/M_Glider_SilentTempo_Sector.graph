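# Translucent unlit material for Niagara mesh particles in the Silent Tempo glider's "Sector" effect.
# Panning noise and plasma textures, distorted by Perlin noise and optionally blended with scrolling smoke, form the animated pattern.
# Curve-atlas rows sampled along U and V mask that pattern, and another row colors it for emissive.
# Dynamic parameters drive opacity erosion and a toggleable vertical vertex offset that falls off along U, with depth fade softening intersections.
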Material /BRCosmetics/Gadgets/Assets/VinderTech_GliderChute/Glider_SilentTempo/FX/M_Glider_SilentTempo_Sector
graph {
 properties {
  bUsedWithNiagaraMeshParticles=true
  blendMode=BLEND_Translucent
  pixelDepthOffsetMode=PDOM_Legacy
  shadingModel=MSM_Unlit
  usedShadingModels=Unlit
 }
 nodes [
  n1 = ParticleColor()
  n2 = MaterialFunctionCall(Float2=TextureCoordinate()) {materialFunction=&BreakOutFloat2Components}
  n3 = CurveAtlasRowParameter(CurveTime=@n2.R) {
   atlas=/Game/Athena/LinearColorCurves/CA_Contrail_SilentTempo
   curve=/Game/Gadgets/Assets/VinderTech_GliderChute/Glider_SilentTempo/FX/Curve_Glider_SilentTempo_03_Peril
   defaultValue=6
   parameterName=CurveA
  }
  n4 = CurveAtlasRowParameter(CurveTime=@n2.G) {
   atlas=&CA_Contrail_SilentTempo
   curve=&Curve_Glider_SilentTempo_03_Peril
   defaultValue=6
   parameterName=CurveB
  }
  n5 = Multiply(A=@n3.G, B=@n4.B)
  n6 = Multiply(A=TextureCoordinate(), B=Constant2Vector() {g=1, r=0.2})
  n7 = Add(A=Time(), B=ParticleRandom())
  n8 = Constant2Vector() {g=0.3, r=0.7}
  n9 = ComponentMask(None=TextureSample(UVs=Panner(Coordinate=@n6, Time=@n7, Speed=@n8)) {samplerType=SAMPLERTYPE_LinearGrayscale, texture=&T_SimplePerlin_Greyscale}.RGB) {g=true, r=true}
  n10 = Multiply(
   A=Add(A=TextureCoordinate(), B=Divide(A=@n9) {constB=5})
   B=ComponentMask(None=VectorParameter() {parameterName=UV, defaultValue=<rgba 0.3 1.6 0 1>}.RGB) {g=true, r=true}
  )
  n11 = Add(A=Time(), B=ParticleRandom())
  n12 = Constant2Vector() {g=0.2, r=1.1}
  n13 = Constant2Vector() {g=0.3, r=0.8}
  n14 = Multiply(
   A=TextureSample(UVs=Panner(Coordinate=@n10, Time=@n11, Speed=@n12)) {samplerType=SAMPLERTYPE_LinearGrayscale, texture=/Game/Effects/Fort_Effects/Emotes/UltraLight/Texture/T_DirectionalNoise_Greyscale}.R
   B=TextureSample(UVs=Panner(Coordinate=@n10, Time=@n11, Speed=@n13)) {samplerType=SAMPLERTYPE_LinearColor, texture=/Game/Effects/Fort_Effects/Emotes/UltraLight/Texture/T_Plasma_D_C}.R
  )
  n15 = TextureCoordinate()
  n16 = Constant2Vector() {g=1, r=0.5}
  n17 = Panner(Coordinate=Add(A=Multiply(A=@n15, B=@n16), B=Constant2Vector() {g=0.15}), Time=@n11, Speed=Constant2Vector() {r=3})
  n18 = LinearInterpolate(
   A=@n14
   B=Multiply(
    A=@n14
    B=TextureSample(UVs=@n17) {samplerType=SAMPLERTYPE_Grayscale, texture=/Game/Effects/Fort_Effects/Textures/Smoke/T_Smoke_Trail_Soft_NoClamp}.R
   )
   Alpha=ScalarParameter() {parameterName=Use Smoke}
  )
  n19 = TextureCoordinate()
  n20 = TextureSampleParameter2D(UVs=Multiply(A=@n19, B=Constant2Vector() {g=1, r=0.5})) {
   parameterName=Erode Tex
   samplerType=SAMPLERTYPE_LinearGrayscale
   texture=&T_SimplePerlin_Greyscale
  }
  n21 = ScalarParameter() {defaultValue=1, group=Erode, parameterName=Erode Edge Hardness, sliderMin=0.5}
  n22 = Multiply(
   A=LinearInterpolate(A=ComponentMask(None=@n19) {r=true}, B=@n20.R, Alpha=ScalarParameter() {defaultValue=0.5, group=Erode, parameterName=Lerp Erode Tex})
   B=@n21
  )
  n23 = DynamicParameter() {paramNames [Erode,WPO,Param3,Param4]}
  n24 = LinearInterpolate(A=@n21, Alpha=@n23.Erode) {constB=-1}
  n25 = Multiply(A=Multiply(A=@n18, B=ScalarParameter() {defaultValue=1, parameterName=Alpha Mult}), B=Saturate(None=Subtract(A=@n22, B=@n24)))
  DepthFade(Opacity=Multiply(A=@n1.A, B=Multiply(A=@n5, B=@n25)), FadeDistance=ScalarParameter() {parameterName=FadeDistance})
  n26 = CurveAtlasRowParameter(CurveTime=Multiply(A=@n5, B=@n18)) {
   atlas=&CA_Contrail_SilentTempo
   curve=/Game/Athena/LinearColorCurves/Curve_Contrail_SilentTempo_04
   defaultValue=3
   parameterName=CurveColor
  }
  n27 = ScalarParameter() {defaultValue=5, parameterName=Emissive Mult}
  MaterialFunctionCall(Emissive=Multiply(A=Multiply(A=@n26, B=@n27), B=@n1.RGB)) {materialFunction=&MF_FX_EmissiveAdjustments}
  n28 = Power(Base=OneMinus(None=ComponentMask(None=@n19) {r=true})) {constExponent=3}
  n29 = Constant3Vector() {constant=<rgba 0 0 1 0>}
  n30 = ScalarParameter() {defaultValue=40, parameterName=WPO}
  Multiply(A=Multiply(A=Multiply(A=@n28, B=@n29), B=Multiply(A=@n23.WPO, B=@n30)), B=ScalarParameter() {parameterName=UseWPO})
 ]
}
