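# Unlit, two-sided translucent trail material for the Silent Tempo glider, usable on Niagara mesh particles.
# Perlin-distorted UVs drive two panned noise and plasma textures, whose product blends two colors and shapes opacity.
# A mask of U times a curve-atlas lookup along V shapes both color and alpha.
# Particle color tints the result, and a dynamic parameter drives erosion through a U gradient mixed with an erode texture.
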Material /BRCosmetics/Gadgets/Assets/VinderTech_GliderChute/Glider_SilentTempo/FX/M_Glider_SilentTempo_Trail
graph {
 properties {
  bUsedWithNiagaraMeshParticles=true
  blendMode=BLEND_Translucent
  pixelDepthOffsetMode=PDOM_Legacy
  shadingModel=MSM_Unlit
  twoSided=true
  usedShadingModels=Unlit
 }
 nodes [
  n1 = MaterialFunctionCall(Float2=TextureCoordinate()) {materialFunction=&BreakOutFloat2Components}
  n2 = CurveAtlasRowParameter(CurveTime=@n1.G) {
   atlas=/Game/Athena/LinearColorCurves/CA_Contrail_SilentTempo
   curve=/Game/Gadgets/Assets/VinderTech_GliderChute/Glider_SilentTempo/FX/Curve_Glider_SilentTempo_03_Peril
   defaultValue=6
   parameterName=CurveB
  }
  n3 = Multiply(A=@n1.R, B=@n2.B)
  n4 = Multiply(A=TextureCoordinate(), B=Constant2Vector() {g=1, r=0.2})
  n5 = Constant2Vector() {g=0.3, r=0.7}
  n6 = ComponentMask(None=TextureSample(UVs=Panner(Coordinate=@n4, Speed=@n5)) {samplerType=SAMPLERTYPE_LinearGrayscale, texture=&T_SimplePerlin_Greyscale}.RGB) {g=true, r=true}
  n7 = Multiply(A=Add(A=TextureCoordinate(), B=Divide(A=@n6) {constB=10}), B=Constant2Vector() {g=1, r=0.5})
  n8 = Constant2Vector() {r=2}
  n9 = Constant2Vector() {r=1.5}
  n10 = Multiply(
   A=TextureSample(UVs=Panner(Coordinate=@n7, Speed=@n8)) {texture=&T_NoisePack_01}.R
   B=TextureSample(UVs=Panner(Coordinate=@n7, Speed=@n9)) {samplerType=SAMPLERTYPE_LinearColor, texture=/Game/Effects/Fort_Effects/Emotes/UltraLight/Texture/T_Plasma_D_C}.R
  )
  n11 = LinearInterpolate(
   A=VectorParameter() {parameterName=ColorB, defaultValue=<rgba 0 0.0583 0.5 12>}.RGB
   B=VectorParameter() {parameterName=ColorA, defaultValue=<rgba 0.2 1.7 2 1>}.RGB
   Alpha=Multiply(A=@n3, B=@n10)
  )
  n12 = ScalarParameter() {defaultValue=1, parameterName=Emissive Mult}
  n13 = ParticleColor()
  MaterialFunctionCall(Emissive=Multiply(A=Multiply(A=@n11, B=@n12), B=@n13.RGB)) {materialFunction=&MF_FX_EmissiveAdjustments}
  n14 = Multiply(A=@n10, B=ScalarParameter() {defaultValue=10, parameterName=Alpha Mult})
  n15 = TextureCoordinate()
  n16 = TextureSampleParameter2D(UVs=Multiply(A=@n15, B=Constant2Vector() {g=1, r=0.5})) {
   parameterName=Erode Tex
   samplerType=SAMPLERTYPE_LinearGrayscale
   texture=&T_SimplePerlin_Greyscale
  }
  n17 = LinearInterpolate(A=ComponentMask(None=@n15) {r=true}, B=@n16.R, Alpha=ScalarParameter() {defaultValue=0.5, group=Erode, parameterName=Lerp Erode Tex})
  n18 = ScalarParameter() {defaultValue=1, group=Erode, parameterName=Erode Edge Hardness, sliderMin=0.5}
  n19 = DynamicParameter() {paramNames [Erode,Param2,Param3,Param4]}
  n20 = Saturate(None=Subtract(A=Multiply(A=@n17, B=@n18), B=LinearInterpolate(A=@n18, Alpha=@n19.Erode) {constB=-1}))
  Multiply(A=@n13.A, B=Multiply(A=@n3, B=Multiply(A=@n14, B=@n20)))
 ]
}
